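# Parent material for the RowLiaison Cafe female character body, supporting skeletal, morph and deformer meshes with per-expression shading models including skin, cloth and clear coat.
# Material attributes flow from the shared 2023 CharacterParent PreFX function through a thin-film function, then into the PostFX function for final output.
# Static switches optionally insert custom crests and sequins between them, with sequins using a mask texture and the PreFX tech-art mask.
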Material /BRCosmetics/Characters/Player/Female/Medium/Bodies/F_MED_RowLiaison_Cafe/Materials/M_RowLiaisonCafe_Parent
graph {
 properties {
  bUseMaterialAttributes=true
  bUsedWithMeshDeformer=true
  bUsedWithMorphTargets=true
  bUsedWithSkeletalMesh=true
  pixelDepthOffsetMode=PDOM_Legacy
  shadingModel=MSM_FromMaterialExpression
  subsurfaceProfile=&SS_HeroSkin_02
  usedShadingModels=Default Lit | Subsurface | Clear Coat | Subsurface Profile | Cloth
 }
 nodes [
  n1 = TextureObjectParameter() {
   group=MF_Sequin
   parameterName=Sequin Mask
   samplerType=SAMPLERTYPE_Masks
   texture=/Game/Characters/Player/Female/Medium/Bodies/F_MED_SkeleProbe/Skins/Candy/Textures/T_SkeleProbe_Candy_Body_FX
  }
  n2 = MaterialFunctionCall() {materialFunction=&MF_CharacterParent_2023_PreFX}
  n3 = MaterialFunctionCall!(Material Attributes=@n2.Result, Fresnel=@n2.Result, Alpha=@n2.Result) {materialFunction=/Game/Characters/Player/Female/Medium/Bodies/F_MED_SkeleProbe/Materials/MF_ThinFilmDuplicate_TechArt}
  n4 = StaticSwitchParameter(
   True=MaterialFunctionCall!(Materials=@n3.Result) {materialFunction=/Game/Weapons/WeaponSkins/MaterialFunction/CustomCrests}.Result
   False=@n3.Result
  ) {group=04_Tech Art, parameterName=useCustomCrest}
  n5 = MaterialFunctionCall!(Mask=@n1, TechArtMask=@n2.Tech Art Mask, inMat=@n4) {materialFunction=/Game/Weapons/WeaponSkins/MaterialFunction/MF_Sequins}
  SetMaterialAttributes!(MaterialAttributes=MaterialFunctionCall(
    Material=StaticSwitchParameter(True=@n5.Result, False=@n4) {group=MF_Sequin, parameterName=EnableSequin, sortPriority=0}
    M=@n2.M
    Pre Skinned Local Position=@n2.M
    Distance Blend=@n2.M
    Fresnel=@n2.M
   ) {materialFunction=&MF_CharacterParent_2023_PostFX}.Result)
 ]
}
