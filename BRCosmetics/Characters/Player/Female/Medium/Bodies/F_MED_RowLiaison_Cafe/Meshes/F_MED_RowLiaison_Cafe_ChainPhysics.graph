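# Physics asset for the hanging hip chain on the F_MED_RowLiaison_Cafe female character body, with capsule bodies on the chain bones and left thigh.
# An adjust bone hangs off the chain base and branches into two-segment front and back strands plus a middle segment, mostly with tight swing limits and zero-degree twist.
# Free-angular constraints tie both strand ends to the middle segment, closing the chain into loops.
# The unconstrained thigh capsules provide a leg collision volume.
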PhysicsAsset /BRCosmetics/Characters/Player/Female/Medium/Bodies/F_MED_RowLiaison_Cafe/Meshes/F_MED_RowLiaison_Cafe_ChainPhysics
bodies {
 Base_hipChain_B [{shapeType=Capsule, length=0.505, radius=0.505, shapeName=Base_hipChain_B_capsule, rotation=<rot 0 -90 0>}]
 Base_hipChain_F [{shapeType=Capsule, length=0.505, radius=0.505, shapeName=Base_hipChain_F_capsule, rotation=<rot 0 -90 0>}]
 adjust_hipChain [{shapeType=Capsule, length=0.505, radius=0.505, shapeName=adjust_hipChain_capsule, rotation=<rot 0 -90 0>}]
 base_hipChain [{shapeType=Capsule, length=0, radius=0.505, shapeName=base_hipChain_capsule, rotation=<rot 0 -90 0>}]
 dyn_hipChain_B01 [{shapeType=Capsule, length=2.2188, radius=2.7812, shapeName=dyn_hipChain_B01_capsule, center=<4 0 0>, rotation=<rot 90 0 0>}]
 dyn_hipChain_B02 [{shapeType=Capsule, length=1.9661, radius=3.3657, shapeName=dyn_hipChain_B02_capsule, center=<4 0 0>, rotation=<rot 90 7.201 0>}]
 dyn_hipChain_F01 [{shapeType=Capsule, length=2.8309, radius=2.1691, shapeName=dyn_hipChain_F01_capsule, center=<4 0 0>, rotation=<rot 90 0 0>}]
 dyn_hipChain_F02 [{shapeType=Capsule, length=1.9475, radius=3.3929, shapeName=dyn_hipChain_F02_capsule, center=<4 0 0>, rotation=<rot 90 -8.5564 0>}]
 dyn_hipChain_M [{shapeType=Capsule, length=10.2889, radius=2.4048, shapeName=dyn_hipChain_M_capsule, center=<5.2518 0 0>, rotation=<rot 90 -90 0>}]
 thigh_l [
  {shapeType=Capsule, length=17.3714, radius=7.4906, center=<-8.3737 -0.0001 2.0835>, rotation=<rot 77.5303 179.9993 179.9999>}
  {shapeType=Capsule, length=0.2584, radius=7.2223, shapeName=thigh_l_capsule, center=<-9.7514 0.6749 1.3746>}
 ]
}
constraints [
 {
  bone1Name=Base_hipChain_B
  bone2Name=adjust_hipChain
  swing1LimitDegrees=45
  swing1Motion=Free
  swing2LimitDegrees=45
  swing2Motion=Free
  twistLimitDegrees=45
  twistMotion=Free
 }
 {bone1Name=dyn_hipChain_B01, bone2Name=Base_hipChain_B, swing1LimitDegrees=10, swing2LimitDegrees=25, twistLimitDegrees=0}
 {bone1Name=dyn_hipChain_B02, bone2Name=dyn_hipChain_B01, swing1LimitDegrees=5, swing2LimitDegrees=35, twistLimitDegrees=0}
 {
  bone1Name=Base_hipChain_F
  bone2Name=adjust_hipChain
  swing1LimitDegrees=25
  swing1Motion=Free
  swing2LimitDegrees=25
  swing2Motion=Free
  twistLimitDegrees=0
  twistMotion=Free
 }
 {bone1Name=dyn_hipChain_F01, bone2Name=Base_hipChain_F, swing1LimitDegrees=10, swing2LimitDegrees=25, twistLimitDegrees=0}
 {bone1Name=dyn_hipChain_F02, bone2Name=dyn_hipChain_F01, swing1LimitDegrees=5, swing2LimitDegrees=25, twistLimitDegrees=0}
 {bone1Name=dyn_hipChain_M, bone2Name=adjust_hipChain, swing1LimitDegrees=8, swing2LimitDegrees=25, twistLimitDegrees=0}
 {bone1Name=adjust_hipChain, bone2Name=base_hipChain, swing1LimitDegrees=15, swing2LimitDegrees=25, twistLimitDegrees=0}
 {
  bone1Name=dyn_hipChain_M
  bone2Name=dyn_hipChain_F02
  swing1LimitDegrees=45
  swing1Motion=Free
  swing2LimitDegrees=45
  swing2Motion=Free
  twistLimitDegrees=45
  twistMotion=Free
 }
 {
  bone1Name=dyn_hipChain_B02
  bone2Name=dyn_hipChain_M
  swing1LimitDegrees=45
  swing1Motion=Free
  swing2LimitDegrees=45
  swing2Motion=Free
  twistLimitDegrees=45
  twistMotion=Free
 }
]
