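# Control Rig for the F_MED_RowLiaison_Cafe body that positions costume accessory bones on top of the shared deform rig function.
# Construction imports the skeleton and adds pelvis-parented thigh-twist controls as rotation references.
# Forward solve runs DeformRig, weighted parent constraints for buckle, gun, pouch, chains and jewellery, then leg correctives.
# Left-thigh swing/twist drives curve-remapped hip-chain offsets, weighted by NotCrouch (default false); right-thigh swing is stored but unused.
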ControlRigBlueprint /BRCosmetics/Characters/Player/Female/Medium/Bodies/F_MED_RowLiaison_Cafe/Meshes/F_MED_RowLiaison_Cafe_CtrlRig
graphs [
 New Function_ContainedGraph {nodes [
   RigVMFunctionEntryNode()
   > &RigUnit_ParentConstraint(
    AdvancedSettings=(RotationOrderForFilter=XZY)
    Child=(Type=Bone,Name="base_buckle")
    Parents="((Item=(Type=Bone,Name=\"thigh_twist_01_r\"),Weight=0.0055),(Item=(Type=Bone,Name=\"deform_groin_r\"),Weight=0.0066),(Item=(Type=Bone,Name=\"thigh_twist_01_l\"),Weight=0.0066),(Item=(Type=Bone,Name=\"deform_groin_l\"),Weight=0.0071),(Item=(Type=Bone,Name=\"spine_02 ...(truncated)"
    Weight=1
   )
   > &RigUnit_ParentConstraint(
    AdvancedSettings=(RotationOrderForFilter=XZY)
    Child=(Type=Bone,Name="base_Gun")
    Parents=((Item=(Type=Bone,Name="deform_glute_r"),Weight=0.0118),(Item=(Type=Bone,Name="thigh_r"),Weight=0.1354),(Item=(Type=Bone,Name="pelvis"),Weight=0.1874),(Item=(Type=Bone,Name="thigh_twist_01_r"),Weight=0.6655))
    Weight=1
   )
   > &RigUnit_ParentConstraint(
    AdvancedSettings=(RotationOrderForFilter=XZY)
    Child=(Type=Bone,Name="base_legPouch")
    Parents=((Item=(Type=Bone,Name="pelvis"),Weight=0.0344),(Item=(Type=Bone,Name="thigh_twist_01_r"),Weight=0.435),(Item=(Type=Bone,Name="thigh_r"),Weight=0.5306))
    Weight=1
   )
   > &RigUnit_ParentConstraint(
    AdvancedSettings=(RotationOrderForFilter=XZY)
    Child=(Type=Bone,Name="attach_hipChain_01")
    Parents=((Item=(Type=Bone,Name="spine_02"),Weight=0.0282),(Item=(Type=Bone,Name="thigh_twist_01_l"),Weight=0.0554),(Item=(Type=Bone,Name="deform_groin_l"),Weight=0.1154),(Item=(Type=Bone,Name="pelvis"),Weight=0.8009))
    Weight=1
   )
   > &RigUnit_ParentConstraint(
    AdvancedSettings=(RotationOrderForFilter=XZY)
    Child=(Type=Bone,Name="attach_hipChain_02")
    Parents=((Item=(Type=Bone,Name="spine_01"),Weight=0.0191),(Item=(Type=Bone,Name="spine_02"),Weight=0.04),(Item=(Type=Bone,Name="deform_glute_l"),Weight=0.0424),(Item=(Type=Bone,Name="pelvis"),Weight=0.8985))
    Weight=1
   )
   > &RigUnit_ParentConstraint(
    AdvancedSettings=(InterpolationType=Average,RotationOrderForFilter=XZY)
    Child=(Type=Bone,Name="Base_hipChain_F")
    Parents=((Item=(Type=Bone,Name="attach_hipChain_01"),Weight=1))
    Weight=1
   )
   > &RigUnit_ParentConstraint(
    AdvancedSettings=(InterpolationType=Average,RotationOrderForFilter=XZY)
    Child=(Type=Bone,Name="Base_hipChain_B")
    Parents=((Item=(Type=Bone,Name="attach_hipChain_02"),Weight=1))
    Weight=1
   )
   > &RigUnit_ParentConstraint(
    AdvancedSettings=(RotationOrderForFilter=XZY)
    Child=(Type=Bone,Name="base_bellyJewellery")
    Parents=((Item=(Type=Bone,Name="spine_01"),Weight=0.0101),(Item=(Type=Bone,Name="spine_03"),Weight=0.0291),(Item=(Type=Bone,Name="spine_04"),Weight=0.0578),(Item=(Type=Bone,Name="spine_02"),Weight=0.3007),(Item=(Type=Bone,Name="pelvis"),Weight=0.6023))
    Weight=1
   )
   > &RigUnit_ParentConstraint(
    AdvancedSettings=(RotationOrderForFilter=XZY)
    Child=(Type=Bone,Name="base_necklace")
    Parents=((Item=(Type=Bone,Name="spine_03"),Weight=0.0394),(Item=(Type=Bone,Name="neck_02"),Weight=0.0236),(Item=(Type=Bone,Name="neck_01"),Weight=0.0943),(Item=(Type=Bone,Name="spine_04"),Weight=0.1808),(Item=(Type=Bone,Name="spine_05"),Weight=0.6519))
    Weight=1
   )
   > &RigUnit_ParentConstraint(
    AdvancedSettings=(RotationOrderForFilter=XZY)
    Child=(Type=Bone,Name="base_zipper")
    Parents=((Item=(Type=Bone,Name="deform_pec_r"),Weight=0.0104),(Item=(Type=Bone,Name="spine_05"),Weight=0.3095),(Item=(Type=Bone,Name="spine_04"),Weight=0.6802))
    Weight=1
   )
   > RigVMFunctionReturnNode()
  ]}
 New Function_ContainedGraph {nodes [
   RigVMFunctionEntryNode()
   > RigVMFunction_Sequence = RigVMAggregateNode()
   RigVMFunctionReturnNode(ExecuteContext=@RigVMFunction_Sequence.D)
   n1 = &RigUnit_GetRelativeTransformForItem(Child=(Type=Bone,Name="thigh_twist_01_l"), Parent=(Type=Control,Name="thigh_twist_01_l_ctrl"), bChildInitial=False, bParentInitial=False)
   n2 = &RigVMFunction_MathQuaternionSwingTwist(TwistAxis=(X=1,Y=0,Z=0), Input=@n1.RelativeTransform.Rotation)
   RigVMVariableNode(Value=&RigVMFunction_MathQuaternionToEuler(RotationOrder=ZYX, Value=@n2.Swing).Result, ExecuteContext=@RigVMFunction_Sequence.A) {variable=Leg_L_Swing}
   > RigVMVariableNode(Value=&RigVMFunction_MathQuaternionToEuler(RotationOrder=ZYX, Value=@n2.Twist).Result) {variable=Leg_L_Twist}
   n3 = &RigUnit_GetRelativeTransformForItem(Child=(Type=Bone,Name="thigh_twist_01_r"), Parent=(Type=Control,Name="thigh_twist_01_r_ctrl"), bChildInitial=False, bParentInitial=False)
   > RigVMVariableNode(Value=&RigVMFunction_MathQuaternionToEuler(
     RotationOrder=ZYX
     Value=&RigVMFunction_MathQuaternionSwingTwist(TwistAxis=(X=1,Y=0,Z=0), Input=@n3.RelativeTransform.Rotation).Swing
    ).Result) {variable=Leg_R_Swing}
   n4 = RigVMVariableNode() {variable=Leg_L_Swing}
   n5 = &RigVMFunction_AnimEvalRichCurve(
    Curve="(EditorCurveData=(Keys=((InterpMode=RCIM_Cubic,TangentMode=RCTM_User,ArriveTangent=0.2796,LeaveTangent=0.2796),(InterpMode=RCIM_Cubic,TangentMode=RCTM_User,Time=0.5729,Value=0.4684,ArriveTangent=1.2761,LeaveTangent=1.2761),(InterpMode=RCIM_Cubic,TangentMod ...(truncated)"
    SourceMaximum=-140
    SourceMinimum=0
    TargetMaximum=-44
    TargetMinimum=0
    Value=@n4.Value.Y
   )
   n6 = &RigVMFunction_AnimEvalRichCurve(
    Curve="(EditorCurveData=(Keys=((InterpMode=RCIM_Cubic,TangentMode=RCTM_User,ArriveTangent=0.2796,LeaveTangent=0.2796),(InterpMode=RCIM_Cubic,TangentMode=RCTM_User,Time=0.5594,Value=0.4668,ArriveTangent=1.2761,LeaveTangent=1.2761),(InterpMode=RCIM_Cubic,TangentMod ...(truncated)"
    SourceMaximum=-140
    SourceMinimum=0
    TargetMaximum=-44
    TargetMinimum=0
    Value=@n4.Value.Z
   )
   n7 = &RigVMFunction_AlphaInterp(
    InterpSpeedDecreasing=10
    InterpSpeedIncreasing=10
    Scale=1
    bClampResult=false
    bInterpResult=false
    bMapRange=true
    Value=&RigVMFunction_MathDoubleAdd(A=@n5.Result, B=@n6.Result).Result
   )
   n8 = &RigVMFunction_MathDoubleMul(B=-4, A=@n7.Result)
   &RigUnit_ModifyTransforms(
    ItemToModify="((Item=(Type=Bone,Name=\"Base_hipChain_F\"),Transform=(Rotation=(X=0,Y=0,Z=0,W=1),Translation=(X=0,Y=0,Z=0),Scale3D=(X=1,Y=1,Z=1))),(Item=(Type=Bone,Name=\"Base_hipChain_B\"),Transform=(Rotation=(X=0,Y=0,Z=0,W=1),Translation=(X=0,Y=0,Z=0),Scale3D=(X=1,Y=1,Z=1) ...(truncated)"
    WeightMinimum=0
    ItemToModify.3.Transform.Translation.Y=&RigVMFunction_MathDoubleMul(B=-0.25, A=@n7.Result).Result
    Weight=&RigVMFunction_MathBoolToFloat(Value=RigVMVariableNode() {variable=NotCrouch}.Value).Result
    ItemToModify.0.Transform.Rotation=&RigVMFunction_MathQuaternionFromEuler(Euler=(X=0,Y=0,Z=0), Euler.X=@n8.Result, Euler.Z=@n8.Result, Euler.Y=&RigVMFunction_MathDoubleMul(B=2.5, A=@n7.Result).Result).Result
    ItemToModify.1.Transform.Rotation=&RigVMFunction_MathQuaternionFromEuler(
     Euler=(X=0,Y=0,Z=0)
     Euler.Y=&RigVMFunction_MathDoubleMul(B=0.85, A=@n7.Result).Result
     Euler.Z=&RigVMFunction_MathDoubleMul(B=-0.6, A=@n7.Result).Result
    ).Result
    ItemToModify.2.Transform.Rotation=&RigVMFunction_MathQuaternionFromEuler(
     Euler=(X=0,Y=0,Z=0)
     Euler.Y=&RigVMFunction_MathDoubleMul(B=0.4, A=@n7.Result).Result
     Euler.Z=&RigVMFunction_MathDoubleMul(B=0.25, A=@n7.Result).Result
    ).Result
    ItemToModify.3.Transform.Rotation=&RigVMFunction_MathQuaternionFromEuler(Euler=(X=0,Y=0,Z=0), Euler.Z=@n8.Result, Euler.X=&RigVMFunction_MathDoubleMul(B=6, A=@n7.Result).Result).Result
    ExecutePin=@RigVMFunction_Sequence.B
   )
   RigVMFunction_MathBoolToFloat = &RigVMFunction_MathBoolToFloat(Value=RigVMVariableNode() {variable=NotCrouch}.Value)
   n9 = &RigVMFunction_AnimEvalRichCurve(
    Curve="(EditorCurveData=(Keys=((InterpMode=RCIM_Cubic,TangentMode=RCTM_User,ArriveTangent=0.2796,LeaveTangent=0.2796),(InterpMode=RCIM_Cubic,TangentMode=RCTM_User,Time=0.5061,Value=0.359,ArriveTangent=1.2761,LeaveTangent=1.2761),(InterpMode=RCIM_Cubic,TangentMode ...(truncated)"
    SourceMaximum=-140
    SourceMinimum=0
    TargetMaximum=-44
    TargetMinimum=0
    Value=RigVMVariableNode() {variable=Leg_L_Twist}.Value.X
   )
   n10 = &RigVMFunction_AlphaInterp(InterpSpeedDecreasing=10, InterpSpeedIncreasing=10, Scale=1, bClampResult=false, bInterpResult=false, bMapRange=true, Value=@n9.Result)
   n11 = &RigVMFunction_MathDoubleMul(B=2.5, A=@n10.Result)
   n12 = &RigVMFunction_MathDoubleMul(B=6, A=@n10.Result)
   > &RigUnit_ModifyTransforms(
    ItemToModify="((Item=(Type=Bone,Name=\"Base_hipChain_B\"),Transform=(Rotation=(X=0,Y=0,Z=0,W=1),Translation=(X=0,Y=0,Z=0),Scale3D=(X=1,Y=1,Z=1))),(Item=(Type=Bone,Name=\"adjust_hipChain\"),Transform=(Rotation=(X=0,Y=0,Z=0,W=1),Translation=(X=0,Y=0,Z=0),Scale3D=(X=1,Y=1,Z=1) ...(truncated)"
    WeightMinimum=0
    Weight=@RigVMFunction_MathBoolToFloat.Result
    ItemToModify.0.Transform.Rotation=&RigVMFunction_MathQuaternionFromEuler(Euler=(X=0,Y=0,Z=0), Euler.Y=@n11.Result, Euler.X=@n12.Result, Euler.Z=@n12.Result).Result
    ItemToModify.1.Transform.Rotation=&RigVMFunction_MathQuaternionFromEuler(Euler=(X=0,Y=0,Z=0), Euler.Y=@n11.Result).Result
   )
   n13 = RigVMVariableNode() {variable=Leg_L_Swing}
   n14 = &RigVMFunction_AnimEvalRichCurve(
    Curve="(EditorCurveData=(Keys=((InterpMode=RCIM_Cubic,TangentMode=RCTM_User,ArriveTangent=0.2796,LeaveTangent=0.2796),(InterpMode=RCIM_Cubic,TangentMode=RCTM_User,Time=0.5233,Value=0.5785,ArriveTangent=1.2761,LeaveTangent=1.2761),(InterpMode=RCIM_Cubic,TangentMod ...(truncated)"
    SourceMaximum=140
    SourceMinimum=0
    TargetMaximum=-44
    TargetMinimum=0
    Value=@n13.Value.Y
   )
   n15 = &RigVMFunction_AnimEvalRichCurve(
    Curve="(EditorCurveData=(Keys=((InterpMode=RCIM_Cubic,TangentMode=RCTM_User,ArriveTangent=0.2796,LeaveTangent=0.2796),(InterpMode=RCIM_Cubic,TangentMode=RCTM_User,Time=0.5554,Value=0.5798,ArriveTangent=1.2761,LeaveTangent=1.2761),(InterpMode=RCIM_Cubic,TangentMod ...(truncated)"
    SourceMaximum=140
    SourceMinimum=0
    TargetMaximum=-44
    TargetMinimum=0
    Value=@n13.Value.Z
   )
   n16 = &RigVMFunction_AlphaInterp(
    InterpSpeedDecreasing=10
    InterpSpeedIncreasing=10
    Scale=1
    bClampResult=false
    bInterpResult=false
    bMapRange=true
    Value=&RigVMFunction_MathDoubleAdd(A=@n14.Result, B=@n15.Result).Result
   )
   n17 = &RigVMFunction_MathDoubleMul(B=0.25, A=@n16.Result)
   n18 = &RigVMFunction_MathQuaternionFromEuler(Euler=(X=0,Y=0,Z=0), Euler.X=@n17.Result, Euler.Y=@n17.Result, Euler.Z=@n17.Result)
   > &RigUnit_ModifyTransforms(
    ItemToModify="((Item=(Type=Bone,Name=\"Base_hipChain_B\"),Transform=(Rotation=(X=0,Y=0,Z=0,W=1),Translation=(X=0,Y=0,Z=0),Scale3D=(X=1,Y=1,Z=1))),(Item=(Type=Bone,Name=\"adjust_hipChain\"),Transform=(Rotation=(X=0,Y=0,Z=0,W=1),Translation=(X=0,Y=0,Z=0),Scale3D=(X=1,Y=1,Z=1) ...(truncated)"
    WeightMinimum=0
    ItemToModify.2.Transform.Translation.Z=&RigVMFunction_MathDoubleMul(B=-0.25, A=@n16.Result).Result
    Weight=&RigVMFunction_MathBoolToFloat(Value=RigVMVariableNode() {variable=NotCrouch}.Value).Result
    ItemToModify.0.Transform.Rotation=@n18.Result
    ItemToModify.1.Transform.Rotation=@n18.Result
   )
  ]}
 RigVMFunctionLibrary {nodes [
   Constraints = RigVMCollapseNode(ExecuteContext=())
   Leg_Accessories_Setup = RigVMCollapseNode(ExecuteContext=())
  ]}
 RigVMFunction_Sequence_1_ContainedGraph {nodes [
   n1 = RigVMFunctionEntryNode()
   RigVMFunction_Sequence = &RigVMFunction_Sequence(ExecuteContext=@n1.ExecuteContext)
   n2 = &RigVMFunction_Sequence(ExecuteContext=@RigVMFunction_Sequence.B)
   RigVMFunctionReturnNode(A=@RigVMFunction_Sequence.A, B=@n2.A, D=@n2.B)
  ]}
 RigVMModel {nodes [
   n1 = &RigUnit_PrepareForExecution()
   &RigUnit_HierarchyImportFromSkeleton(NameSpace=None, bIncludeCurves=False, bIncludeMeshSockets=False, bIncludeVirtualBones=True, ExecutePin=@n1.ExecutePin)
   > Control Stack at Item = RigVMFunctionReferenceNode(
    Add Bottom Null=False
    Add Null=False
    Add Secondary Control=False
    Control Suffix=_ctrl
    Item=(Name="thigh_twist_01_l",Type=Bone)
    Null Suffix=_null
    Parent=(Name="pelvis",Type=Bone)
    Secondary Settings="(Shape=(bVisible=True,Name=\"Default\",Color=(R=1,G=0,B=0,A=1),Transform=(Rotation=(X=0,Y=0,Z=0,W=1),Translation=(X=0,Y=0,Z=0),Scale3D=(X=1,Y=1,Z=1))),Proxy=(bIsProxy=False,ShapeVisibility=BasedOnSelection),DisplayName=\"None\",InitialSpace=LocalSpace,bUsePref ...(truncated)"
    Settings="(Shape=(bVisible=True,Name=\"Default\",Color=(R=0,G=0,B=1,A=1),Transform=(Rotation=(X=0,Y=0,Z=0,W=1),Translation=(X=0,Y=0,Z=0),Scale3D=(X=1,Y=1,Z=1))),Proxy=(bIsProxy=False,ShapeVisibility=BasedOnSelection),DisplayName=\"None\",InitialSpace=LocalSpace,bUsePref ...(truncated)"
    Tag Controls=true
    Zero Controls=true
   )
   > Control Stack at Item_1 = RigVMFunctionReferenceNode(
    Add Bottom Null=False
    Add Null=False
    Add Secondary Control=False
    Control Suffix=_ctrl
    Item=(Name="thigh_twist_01_r",Type=Bone)
    Null Suffix=_null
    Parent=(Name="pelvis",Type=Bone)
    Secondary Settings="(Shape=(bVisible=True,Name=\"Default\",Color=(R=1,G=0,B=0,A=1),Transform=(Rotation=(X=0,Y=0,Z=0,W=1),Translation=(X=0,Y=0,Z=0),Scale3D=(X=1,Y=1,Z=1))),Proxy=(bIsProxy=False,ShapeVisibility=BasedOnSelection),DisplayName=\"None\",InitialSpace=LocalSpace,bUsePref ...(truncated)"
    Settings="(Shape=(bVisible=True,Name=\"Default\",Color=(R=0,G=0,B=1,A=1),Transform=(Rotation=(X=0,Y=0,Z=0,W=1),Translation=(X=0,Y=0,Z=0),Scale3D=(X=1,Y=1,Z=1))),Proxy=(bIsProxy=False,ShapeVisibility=BasedOnSelection),DisplayName=\"None\",InitialSpace=LocalSpace,bUsePref ...(truncated)"
    Tag Controls=true
    Zero Controls=true
   )
   RigUnit_BeginExecution = &RigUnit_BeginExecution()
   FN_DeformRig_FNC = RigVMFunctionReferenceNode(
    DeformRig_ Curve="(deformRig_l_upperarm_up_alpha_72_#1=False,deformRig_r_upperarm_up_alpha_73_#2=False,deformRig_l_upperarm_fr_alpha_74_#3=False,deformRig_r_upperarm_fr_alpha_75_CB51B8 ...(truncated)"
    DeformRig_Bool="(upperarm_1_#4=True,elbow_3_#5=True,wrist_7_#6=True,groin_8_#7=True,glute_17_#8=True,knee_9_40C408324CA261 ...(truncated)"
    DeformRig_Debug=(upperarm_debug_18_#4=False,wrist_debug_20_#6=False,lat_debug_26_#9=False,groinGlute_debug_27_#7=False)
    DeformRig_Value="(upperarm_up_18_#10=(Rotation=(X=0.0,Y=-0.0,Z=0.0,W=1.0),Translation=(X=0.0,Y=0.0,Z=8.0),Scale3D=(X=0.70,Y=1.0,Z=1.0)),upperarm_fwd_fr_19_#11=(Rotation=(X=0.0,Y=-0.0,Z=0.0,W=1.0),Translation=(X=0.0, ...(truncated)"
    ExecuteContext=@RigUnit_BeginExecution.ExecutePin
   )
   > Constraints = RigVMFunctionReferenceNode()
   > Leg_Accessories_Setup = RigVMFunctionReferenceNode()
  ]}
]
variables [
 Crouch {type=bool, default=False}
 Leg_L_Swing {type=FVector, default=(X=0,Y=0,Z=0)}
 Leg_L_Twist {type=FVector, default=(X=0,Y=0,Z=0)}
 Leg_R_Swing {type=FVector, default=(X=0,Y=0,Z=0)}
 NotCrouch {type=bool, default=False}
]
