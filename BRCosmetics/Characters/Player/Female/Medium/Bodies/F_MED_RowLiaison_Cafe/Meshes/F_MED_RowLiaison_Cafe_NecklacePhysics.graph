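# Physics asset for the necklace and zipper of the F_MED_RowLiaison_Cafe female player body cosmetic.
# Two large spine capsules cover the torso, and tiny capsules on base_necklace and base_zipper link to them through free-angle constraints.
# A two-segment necklace chain and a single zipper pull hang from those bases.
# Their constraints restrict swing and allow zero twist, with the necklace's swing asymmetric and slightly looser at the second segment.
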PhysicsAsset /BRCosmetics/Characters/Player/Female/Medium/Bodies/F_MED_RowLiaison_Cafe/Meshes/F_MED_RowLiaison_Cafe_NecklacePhysics
bodies {
 base_necklace [{shapeType=Capsule, length=0, radius=0.505, shapeName=base_necklace_capsule, rotation=<rot 0 -90 0>}]
 base_zipper [{
   shapeType=Capsule
   length=0
   radius=0.505
   shapeName=base_zipper_capsule
   center=<-0.1994 0.0024 -0.3165>
   rotation=<rot 0 -89.6811 -88.3732>
  }]
 dyn_Necklace_01 [{shapeType=Capsule, length=4.1204, radius=0.8796, shapeName=dyn_Necklace_01_capsule, center=<2.354 0 0>, rotation=<rot 0 0 90>}]
 dyn_Necklace_02 [{shapeType=Capsule, length=4.0646, radius=0.9354, shapeName=dyn_Necklace_02_capsule, center=<4 0 0>, rotation=<rot 0 0 90>}]
 dyn_zipper [{shapeType=Capsule, length=1.4201, radius=0.7425, shapeName=dyn_zipper_capsule, center=<1.5 0 0>, rotation=<rot 90 0 0>}]
 spine_04 [{shapeType=Capsule, length=2.5104, radius=9.629, shapeName=spine_04_capsule, center=<-0.8138 2.4881 0.0021>}]
 spine_05 [{shapeType=Capsule, length=8.3285, radius=5.6176, shapeName=spine_05_capsule, center=<2.5825 0.3493 -0.0871>}]
}
constraints [
 {
  bone1Name=spine_05
  bone2Name=spine_04
  swing1LimitDegrees=45
  swing1Motion=Free
  swing2LimitDegrees=45
  swing2Motion=Free
  twistLimitDegrees=45
  twistMotion=Free
 }
 {
  bone1Name=base_necklace
  bone2Name=spine_05
  swing1LimitDegrees=45
  swing1Motion=Free
  swing2LimitDegrees=45
  swing2Motion=Free
  twistLimitDegrees=45
  twistMotion=Free
 }
 {bone1Name=dyn_Necklace_01, bone2Name=base_necklace, swing1LimitDegrees=5, swing2LimitDegrees=20, twistLimitDegrees=0}
 {bone1Name=dyn_Necklace_02, bone2Name=dyn_Necklace_01, swing1LimitDegrees=8, swing2LimitDegrees=25, twistLimitDegrees=0}
 {
  bone1Name=base_zipper
  bone2Name=spine_04
  swing1LimitDegrees=45
  swing1Motion=Free
  swing2LimitDegrees=45
  swing2Motion=Free
  twistLimitDegrees=45
  twistMotion=Free
 }
 {bone1Name=dyn_zipper, bone2Name=base_zipper, swing1LimitDegrees=25, swing2LimitDegrees=25, twistLimitDegrees=0}
]
