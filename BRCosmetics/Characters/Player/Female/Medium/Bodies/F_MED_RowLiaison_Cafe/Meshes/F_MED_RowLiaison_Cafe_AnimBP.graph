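# Post-process style AnimBlueprint for the F_MED_RowLiaison_Cafe female cosmetic body, animating its hip chain and necklace.
# The copied parent pose passes through a Control Rig and necklace RigidBody into a cached pose.
# Crouch, slide, emote or DBNO selects weighted constraints with additive bone offsets; otherwise half-alpha RigidBody simulation drives the hip chain.
# The EventGraph updates state flags, gravity overrides and WPO cloth material parameters, varying wind with skydiving.
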AnimBlueprint /BRCosmetics/Characters/Player/Female/Medium/Bodies/F_MED_RowLiaison_Cafe/Meshes/F_MED_RowLiaison_Cafe_AnimBP
target_skeleton=&/F_MED_RowLiaison_Cafe_Skeleton
graphs [
 AnimGraph {nodes [
   n1 = "Constraint - dyn_hipChain_F02 (PPPP)"!(
    ComponentPose=Local To Component(LocalPose=Use cached pose 'startPose'())
    Alpha=1
    ConstraintWeights_0=0.3596
    ConstraintWeights_1:float=0.324
    ConstraintWeights_2:float=0.2916
    ConstraintWeights_3:float=0.0248
   ) -> (Pose) {
    boneToModify {boneName=dyn_hipChain_F02}
    constraintSetup [
     {targetBone {boneName=pelvis}}
     {targetBone {boneName=deform_groin_l}}
     {targetBone {boneName=thigh_twist_01_l}}
     {targetBone {boneName=thigh_l}}
    ]
    constraintWeights [1,1,1,1]
   }
   n2 = "Constraint - dyn_hipChain_M (PPPPP)"!(
    ComponentPose=@n1
    Alpha=1
    ConstraintWeights_0=0.5893
    ConstraintWeights_1:float=0.2519
    ConstraintWeights_2:float=0.0673
    ConstraintWeights_3:float=0.0459
    ConstraintWeights_4:float=0.0456
   ) -> (Pose) {
    boneToModify {boneName=dyn_hipChain_M}
    constraintSetup [
     {targetBone {boneName=thigh_twist_01_l}}
     {targetBone {boneName=pelvis}}
     {targetBone {boneName=thigh_l}}
     {targetBone {boneName=deform_glute_l}}
     {targetBone {boneName=deform_groin_l}}
    ]
    constraintWeights [1,1,1,1,1]
   }
   n3 = "Constraint - dyn_hipChain_B02 (PPPP)"!(
    ComponentPose=@n2
    Alpha=1
    ConstraintWeights_0=0.5953
    ConstraintWeights_1:float=0.2225
    ConstraintWeights_2:float=0.1736
    ConstraintWeights_3:float=0.0085
   ) -> (Pose) {
    boneToModify {boneName=dyn_hipChain_B02}
    constraintSetup [
     {targetBone {boneName=pelvis}}
     {targetBone {boneName=deform_glute_l}}
     {targetBone {boneName=thigh_twist_01_l}}
     {targetBone {boneName=thigh_l}}
    ]
    constraintWeights [1,1,1,1]
   }
   n4 = "Transform (Modify) Bone - Bone: dyn_hipChain_M"!(ComponentPose=@n3, bAlphaBoolEnabled:bool=True, Translation="-0.074,-0.000,0.644", Rotation="0,0,0", Scale="1,1,1") -> (Pose) {
    alphaInputType=Bool
    rotationMode=BMM_Additive
    rotationSpace=BCS_BoneSpace
    translationMode=BMM_Additive
    translationSpace=BCS_BoneSpace
    alphaBoolBlend {blendInTime=0.1, blendOption=QuarticInOut, blendOutTime=0.3}
    boneToModify {boneName=dyn_hipChain_M}
    translation {x=-0.074, z=0.644}
   }
   n5 = "Transform (Modify) Bone - Bone: Base_hipChain_F"!(ComponentPose=@n4, bAlphaBoolEnabled:bool=True, Translation="-0.453,-0.133,0.352", Rotation="74.224,121.927,31.829", Scale="1,1,1") -> (Pose) {
    alphaInputType=Bool
    rotationMode=BMM_Additive
    rotationSpace=BCS_BoneSpace
    translationMode=BMM_Additive
    translationSpace=BCS_BoneSpace
    alphaBoolBlend {blendInTime=0.1, blendOption=QuarticInOut, blendOutTime=0.3}
    boneToModify {boneName=Base_hipChain_F}
    rotation=<rot 74.224 121.9271 31.829>
    translation=<-0.453 -0.133 0.352>
   }
   n6 = "Transform (Modify) Bone - Bone: dyn_hipChain_F02"!(ComponentPose=@n5, bAlphaBoolEnabled:bool=True, Translation="3.657,-6.764,-3.910", Rotation="39.462,-27.626,36.417", Scale="1,1,1") -> (Pose) {
    alphaInputType=Bool
    rotationMode=BMM_Additive
    rotationSpace=BCS_BoneSpace
    translationMode=BMM_Additive
    translationSpace=BCS_BoneSpace
    alphaBoolBlend {blendInTime=0.1, blendOption=QuarticInOut, blendOutTime=0.3}
    boneToModify {boneName=dyn_hipChain_F02}
    rotation=<rot 39.462 -27.626 36.417>
    translation=<3.657 -6.764 -3.91>
   }
   n7 = "Transform (Modify) Bone - Bone: dyn_hipChain_B02"!(ComponentPose=@n6, bAlphaBoolEnabled:bool=True, Translation="0,-2,0", Rotation="0,0,0", Scale="1,1,1") -> (Pose) {
    alphaInputType=Bool
    rotationMode=BMM_Additive
    rotationSpace=BCS_BoneSpace
    translationMode=BMM_Additive
    translationSpace=BCS_BoneSpace
    alphaBoolBlend {blendInTime=0.1, blendOption=QuarticInOut, blendOutTime=0.3}
    boneToModify {boneName=dyn_hipChain_B02}
    translation {y=-2}
   }
   n8 = "Transform (Modify) Bone - Bone: Base_hipChain_F"(ComponentPose=@n7, bAlphaBoolEnabled=True, Translation="0,0,0", Rotation="5,1.7,0", Scale="1,1,1") {
    alphaInputType=Bool
    rotationMode=BMM_Additive
    rotationSpace=BCS_BoneSpace
    translationMode=BMM_Additive
    translationSpace=BCS_BoneSpace
    boneToModify {boneName=Base_hipChain_F}
    rotation {pitch=5, yaw=1.7}
   }
   n9 = "Transform (Modify) Bone - Bone: dyn_hipChain_F02"(ComponentPose=@n8, bAlphaBoolEnabled=True, Translation="0,3.8,0", Rotation="0,34,4.15", Scale="1,1,1") {
    alphaInputType=Bool
    rotationMode=BMM_Additive
    rotationSpace=BCS_BoneSpace
    translationMode=BMM_Additive
    translationSpace=BCS_BoneSpace
    boneToModify {boneName=dyn_hipChain_F02}
    rotation {roll=4.15, yaw=34}
    translation {y=3.8}
   }
   n10 = "Transform (Modify) Bone - Bone: adjust_hipChain"!(ComponentPose=@n9, bAlphaBoolEnabled:bool=True, Translation="0,0,0", Rotation="8,0,0", Scale="1,1,1") -> (Pose) {
    alphaInputType=Bool
    rotationMode=BMM_Additive
    rotationSpace=BCS_BoneSpace
    translationMode=BMM_Additive
    translationSpace=BCS_BoneSpace
    boneToModify {boneName=adjust_hipChain}
    rotation {pitch=8}
   }
   n11 = "Transform (Modify) Bone - Bone: Base_hipChain_F"(ComponentPose=@n10, bAlphaBoolEnabled=True, Translation="0,0,0", Rotation="0,25,0", Scale="1,1,1") {
    alphaInputType=Bool
    rotationMode=BMM_Additive
    rotationSpace=BCS_BoneSpace
    translationMode=BMM_Additive
    translationSpace=BCS_BoneSpace
    boneToModify {boneName=Base_hipChain_F}
    rotation {yaw=25}
   }
   n12 = "Transform (Modify) Bone - Bone: dyn_hipChain_F02"(ComponentPose=@n11, bAlphaBoolEnabled=True, Translation="-6,-8,0", Rotation="0,-61.8928,0", Scale="1,1,1") {
    alphaInputType=Bool
    rotationMode=BMM_Additive
    rotationSpace=BCS_BoneSpace
    translationMode=BMM_Additive
    translationSpace=BCS_BoneSpace
    boneToModify {boneName=dyn_hipChain_F02}
    rotation {yaw=-61.8928}
    translation=<-6 -8>
   }
   n13 = "Transform (Modify) Bone - Bone: dyn_hipChain_F02"(ComponentPose=@n12, bAlphaBoolEnabled=True, Translation="0,-6,0", Rotation="0,0,0", Scale="1,1,1") {
    alphaInputType=Bool
    rotationMode=BMM_Additive
    rotationSpace=BCS_BoneSpace
    translationMode=BMM_Additive
    translationSpace=BCS_BoneSpace
    boneToModify {boneName=dyn_hipChain_F02}
    translation {y=-6}
   }
   n14 = "Transform (Modify) Bone - Bone: Base_hipChain_F"(ComponentPose=@n13, bAlphaBoolEnabled=True, Translation="0,0,0", Rotation="33.110,17.963,9.405", Scale="1,1,1") {
    alphaInputType=Bool
    rotationMode=BMM_Additive
    rotationSpace=BCS_BoneSpace
    translationMode=BMM_Additive
    translationSpace=BCS_BoneSpace
    boneToModify {boneName=Base_hipChain_F}
    rotation=<rot 33.11 17.9629 9.405>
   }
   n15 = "Transform (Modify) Bone - Bone: dyn_hipChain_F02"(ComponentPose=@n14, bAlphaBoolEnabled=True, Translation="10.616,0.824,2.876", Rotation="-13.645,-21.175,-73.119", Scale="1,1,1") {
    alphaInputType=Bool
    rotationMode=BMM_Additive
    rotationSpace=BCS_BoneSpace
    translationMode=BMM_Additive
    translationSpace=BCS_BoneSpace
    boneToModify {boneName=dyn_hipChain_F02}
    rotation=<rot -13.645 -21.175 -73.119>
    translation=<10.616 0.824 2.876>
   }
   n16 = "Transform (Modify) Bone - Bone: dyn_hipChain_B02"(ComponentPose=@n15, bAlphaBoolEnabled=True, Translation="0.002,-2.440,-0.236", Rotation="0,0,0", Scale="1,1,1") {
    alphaInputType=Bool
    rotationMode=BMM_Additive
    rotationSpace=BCS_BoneSpace
    translationMode=BMM_Additive
    translationSpace=BCS_BoneSpace
    alphaBoolBlend {blendInTime=0.1, blendOption=QuarticInOut, blendOutTime=0.3}
    boneToModify {boneName=dyn_hipChain_B02}
    translation=<0.002 -2.44 -0.236>
   }
   n17 = RigidBody!(
    ComponentPose=Local To Component(LocalPose=Use cached pose 'startPose'())
    Alpha=0.5
    OverrideWorldGravity:Vector="0,0,0"
    ExternalForce
    SimSpaceSettings:Sim Space Settings Structure=(WorldAlpha=0,VelocityScaleZ=1,DampingAlpha=1,MaxLinearVelocity=10000,MaxAngularVelocity=10000,MaxLinearAcceleration=10000,MaxAngularAcceleration=10000,ExternalLinearDragV=(X=0,Y=0,Z=0),ExternalLinearVelocity=(X=0,Y=0,Z=0),ExternalAngularVelocity=(X=0,Y=0,Z=0))
   ) -> (Pose) {
    bClampLinearTranslationLimitToRefPose=true
    bOverrideWorldGravity=true
    overridePhysicsAsset=&/F_MED_RowLiaison_Cafe_ChainPhysics
    simulationSpace=BaseBoneSpace
    baseBoneRef {boneName=pelvis}
   }
   n18 = "Transform (Modify) Bone - Bone: base_hipChain"!(ComponentPose=@n17, bAlphaBoolEnabled:bool=True, Translation="0,0,0", Rotation="8,0,0", Scale="1,1,1") -> (Pose) {
    alphaInputType=Bool
    rotationMode=BMM_Additive
    rotationSpace=BCS_BoneSpace
    translationMode=BMM_Additive
    translationSpace=BCS_BoneSpace
    boneToModify {boneName=base_hipChain}
    rotation {pitch=8}
   }
   n19 = "Transform (Modify) Bone - Bone: base_hipChain"(ComponentPose=@n18, bAlphaBoolEnabled=True, Translation="0,0,0", Rotation="4,0,0", Scale="1,1,1") {
    alphaInputType=Bool
    rotationMode=BMM_Additive
    rotationSpace=BCS_BoneSpace
    translationMode=BMM_Additive
    translationSpace=BCS_BoneSpace
    boneToModify {boneName=base_hipChain}
    rotation {pitch=4}
   }
   n20 = "Transform (Modify) Bone - Bone: Base_hipChain_F"(ComponentPose=@n19, bAlphaBoolEnabled=True, Translation="0,0,0", Rotation="0,0,0", Scale="1,1,1") {
    alphaInputType=Bool
    rotationMode=BMM_Additive
    rotationSpace=BCS_BoneSpace
    translationMode=BMM_Additive
    translationSpace=BCS_BoneSpace
    boneToModify {boneName=Base_hipChain_F}
   }
   n21 = "Transform (Modify) Bone - Bone: adjust_hipChain"(ComponentPose=@n20, bAlphaBoolEnabled=True, Translation="0,0,0", Rotation="-1.5,0,0", Scale="1,1,1") {
    alphaInputType=Bool
    rotationMode=BMM_Additive
    rotationSpace=BCS_BoneSpace
    translationMode=BMM_Additive
    translationSpace=BCS_BoneSpace
    alphaBoolBlend {blendInTime=0.1, blendOutTime=0.3}
    boneToModify {boneName=adjust_hipChain}
    rotation {pitch=-1.5}
   }
   n22 = "Transform (Modify) Bone - Bone: dyn_hipChain_B02"!(ComponentPose=@n21, Alpha:float=1, Translation="0,-3,0", Rotation="0,0,0", Scale="1,1,1") -> (Pose) {
    rotationMode=BMM_Additive
    rotationSpace=BCS_BoneSpace
    translationMode=BMM_Additive
    translationSpace=BCS_BoneSpace
    boneToModify {boneName=dyn_hipChain_B02}
    translation {y=-3}
   }
   n23 = Get FEnoEmote() -> (FEnoEmote:bool)
   n24 = "Transform (Modify) Bone - Bone: base_necklace"!(ComponentPose=@n22, bAlphaBoolEnabled:bool=@n23, Translation="0,0,0", Rotation="-8,0,0", Scale="1,1,1") -> (Pose) {alphaInputType=Bool, rotationMode=BMM_Additive, rotationSpace=BCS_BoneSpace, boneToModify {boneName=base_necklace}, rotation {pitch=-8}}
   n25 = "Transform (Modify) Bone - Bone: dyn_Necklace_02"!(ComponentPose=@n24, bAlphaBoolEnabled:bool=@n23, Translation="0,0,0", Rotation="11.514,0.162,0.157", Scale="1,1,1") -> (Pose) {
    alphaInputType=Bool
    rotationMode=BMM_Additive
    rotationSpace=BCS_BoneSpace
    boneToModify {boneName=dyn_Necklace_02}
    rotation=<rot 11.514 0.162 0.157>
   }
   Output Pose(Result=Blend Poses by bool(
     BlendPose_0=Component To Local(ComponentPose=@n16)
     BlendPose_1=Component To Local(ComponentPose=@n25)
     bActiveValue=Get Crouch_Emote_Slide() -> (Crouch_Emote_Slide:bool)
    )) {layerGroup=None}
   n26 = Control Rig!(Source=Copy Pose From Mesh() {bUseAttachedParent=true}, Alpha, NotCrouch:bool=Get NotCrouching() -> (NotCrouching:bool)) -> (Pose) {controlRigAssetReference {blueprintRigClass=&/F_MED_RowLiaison_Cafe_CtrlRig.F_MED_RowLiaison_Cafe_CtrlRig_C}}
   n27 = RigidBody(
    ComponentPose=Local To Component(LocalPose=@n26)
    OverrideWorldGravity="0,0,0"
    SimSpaceSettings=(WorldAlpha=0,VelocityScaleZ=1,DampingAlpha=1,MaxLinearVelocity=10000,MaxAngularVelocity=10000,MaxLinearAcceleration=10000,MaxAngularAcceleration=10000,ExternalLinearDragV=(X=0,Y=0,Z=0),ExternalLinearVelocity=(X=0,Y=0,Z=0),ExternalAngularVelocity=(X=0,Y=0,Z=0))
   ) {
    bClampLinearTranslationLimitToRefPose=true
    bOverrideWorldGravity=true
    overridePhysicsAsset=&/F_MED_RowLiaison_Cafe_NecklacePhysics
    simulationSpace=BaseBoneSpace
    baseBoneRef {boneName=spine_05}
   }
   Save cached pose 'startPose'(Pose=Component To Local(ComponentPose=@n27)) {cacheName=startPose}
  ]}
 EventGraph {nodes [
   n1 = Event BlueprintUpdateAnimation()
   > n2 = Sequence!(execute) -> (then_0, then_1, then_2:Exec)
   Is Valid(exec=@n2.then_1, InputObject=Get PlayerPawn() -> (PlayerPawn:Fort Player Pawn Athena Object Reference))
   > n3 = Branch(Condition=Get bIsInFrontEnd())
   Set WPO Globals Function(
    execute=@n3.else
    Material="Get Mat Instance Dynamic "() -> ("Mat Instance Dynamic ":Material Instance Dynamic Object Reference)
    Overlay Material="Get Overlya Mat Instance Dynamic "() -> ("Overlya Mat Instance Dynamic ":Material Instance Dynamic Object Reference)
    Skeletal Mesh=Get SkelMesh()
    "Local Noise Transform  Location"="Get Local Noise Transform  Location"() -> ("Local Noise Transform  Location":Vector)
    Local Noise Offset=Get Local Noise Offset_0() -> (Local Noise Offset_0:Vector)
   )
   > Set WPO Regional Function!(
    execute
    Material:Material Instance Dynamic Object Reference="Get Mat Instance Dynamic "()
    Overlay Material:Material Instance Dynamic Object Reference="Get Overlya Mat Instance Dynamic "()
    Skeletal Mesh:Skeletal Mesh Component Object Reference=Get SkelMesh()
    Velocity Transform:Name=foot_r
    Velocity Interp Speed:double=Get Velocity Interp Speed()
    Collision Transform:Name=calf_r
    Collision Transform Axis:EAxis Enum
    Invert Collision Transform Axis:bool=false
    Delta Time X:double=@n1.DeltaTimeX
    Previous Velocity Transform Location:"Vector (by ref)"=Get R1 Velocity Transform Location()
    Previous Velocity Transform Velocity:"Vector (by ref)"=Get R1 Velocity Transform Velocity()
    Region:int=1
   ) -> (then)
   > Set WPO Regional Function(
    Material="Get Mat Instance Dynamic "()
    Overlay Material="Get Overlya Mat Instance Dynamic "()
    Skeletal Mesh=Get SkelMesh()
    Velocity Transform=foot_l
    Velocity Interp Speed=Get Velocity Interp Speed()
    Collision Transform=calf_l
    Invert Collision Transform Axis=true
    Delta Time X=@n1.DeltaTimeX
    Previous Velocity Transform Location=Get R2 Velocity Transform Location()
    Previous Velocity Transform Velocity=Get R2 Velocity Transform Velocity()
    Region=2
   )
   > n4 = Branch(Condition=Get bIsSkyDiving())
   > Set WPO Pho Wind Intensity(WPO Pho Wind Intensity=1)
   > Set WPO Pho Turbulence(WPO Pho Turbulence:double=1) -> (Output_Get:double)
   > n5 = Set WPO Pho Drag(WPO Pho Drag:double=1) -> (Output_Get:double)
   SetScalarParameterValue(execute=[@n8.then, @n5.then], self="Get Mat Instance Dynamic "(), ParameterName=ClothSim_WindStrength, Value=Get WPO Pho Wind Intensity())
   n6 = Get WPO Pho Turbulence() -> (WPO Pho Turbulence:double)
   > SetScalarParameterValue(self="Get Mat Instance Dynamic "(), ParameterName=ClothSim_R1_TurbulenceStrength, Value=@n6)
   > SetScalarParameterValue(self="Get Mat Instance Dynamic "(), ParameterName=ClothSim_R2_TurbulenceStrength, Value=@n6)
   n7 = Get WPO Pho Drag() -> (WPO Pho Drag:double)
   > SetScalarParameterValue(self="Get Mat Instance Dynamic "(), ParameterName=ClothSim_R1_DragStrength, Value=@n7)
   > SetScalarParameterValue(self="Get Mat Instance Dynamic "(), ParameterName=ClothSim_R2_DragStrength, Value=@n7)
   Set WPO Pho Wind Intensity(execute=@n4.else, WPO Pho Wind Intensity=0)
   > Set WPO Pho Turbulence(WPO Pho Turbulence=0.2)
   > n8 = Set WPO Pho Drag(WPO Pho Drag=5)
   n9 = "Gravity Override State Function (Nativized)"!(
    execute=@n2.then_0
    CharacterPartAnimInstance:Custom Character Part Anim Instance Object Reference=Self-Reference()
    GravityOverrideParameterSet:Fort Gravity Override Parameter Set Structure=Get Gravity Override Params() -> (Gravity Override Params:Fort Gravity Override Parameter Set Structure)
    RigidBodyAnimNodeParameterSet:Fort Rigid Body Anim Node Parameter Set Structure=Get RBAN Params() -> (RBAN Params:Fort Rigid Body Anim Node Parameter Set Structure)
    DebugDraw:bool=false
    GameTimeInSeconds:float=GetGameTimeInSeconds()
    DebugJointName:Name=root
    StateLogicLayerName:Name=none
   ) -> (
    then
    OutGravityOverride:Vector
    OutCurrentState:Name=None
    OutGravityOverrideParameters:Fort Gravity Override Parameters Structure
    OutRigidBodyAnimNodeParameters:Fort Rigid Body Anim Node Parameters Structure
   )
   > Set Gravity Override(Gravity Override=@n9)
   > Set RBAN Settings Out(RBAN Settings Out=@n9.OutRigidBodyAnimNodeParameters)
   > Set CrouchAlpha(CrouchAlpha:double=float * float(A=Get CrouchingWithRigidBodyEnabled() -> (CrouchingWithRigidBodyEnabled:float), B=0.45) {is=title K2Node_PromotableOperator}) -> (Output_Get:double)
   > Set NotCrouching(NotCrouching:bool=NOT Boolean(A=Get bIsCrouching())) -> (Output_Get:bool)
   > Set Crouch_Emote_Slide(Crouch_Emote_Slide:bool=OR Boolean!(A=Get bIsCrouching(), B=Get bIsSliding(), C:bool=Get bIsPlayingEmote(), D:bool=Get bIsDBNO()) -> (ReturnValue)) -> (Output_Get:bool)
   n10 = Get bIsPlayingEmote()
   > Set FEnoEmote(FEnoEmote:bool=AND Boolean(A=Get bIsInFrontEnd(), B=NOT Boolean(A=@n10))) -> (Output_Get:bool)
   n11 = GetAnimInstance(self=GetOwnerSkeletalMeshComponent())
   Is Valid(exec=@n2.then_2, InputObject=@n11)
   > Set IsEasternBlocPlaying(IsEasternBlocPlaying:bool=Montage_IsPlaying(self=@n11, Montage=/BRCosmetics/Animation/Game/MainPlayer/Emotes/Eastern_Bloc/Emote_EasternBloc_CMF_M)) -> (Output_Get:bool)
   Event OnCharacterCustomizationLoaded()
   > n12 = Set SkelMesh(SkelMesh=GetPartSkeletalMeshComponent())
   > n13 = Get Dynamic Material Instance(SkeletalMeshComponent=@n12)
   > "Set Mat Instance Dynamic "("Mat Instance Dynamic ":Material Instance Dynamic Object Reference=@n13) -> (Output_Get:Material Instance Dynamic Object Reference)
   > n14 = GetOrCreateDynamicOverlayMaterial(MeshComponent=@n12, MaterialName=None)
   > "Set Overlya Mat Instance Dynamic "("Overlya Mat Instance Dynamic ":Material Instance Dynamic Object Reference=@n14) -> (Output_Get:Material Instance Dynamic Object Reference)
   Event BlueprintInitializeAnimation()
   > Set SkelMesh(SkelMesh=GetOwnerSkeletalMeshComponent())
   n15 = TryGetPawnOwner()
   > Is Valid(InputObject=@n15)
   > n16 = Cast To FortPlayerPawnAthena(Object=@n15)
   > Set PlayerPawn(PlayerPawn:Fort Player Pawn Athena Object Reference=@n16) -> (Output_Get:Fort Player Pawn Athena Object Reference)
  ]}
]
variables [
 Crouch_Emote_Slide {type=boolean}
 CrouchAlpha {type=number}
 CrouchMov {type=boolean}
 CrouchNoMov {type=boolean}
 FEnoEmote {type=boolean}
 Gravity Override {type=Vector}
 Gravity Override Params {
  type=FortGravityOverrideParameterSet
  default=(Skydive_GoingUp=(GravityOverride=(WindFrequency=(X=1.500000,Y=7.000000,Z=8.000000),WindAmplitude=(X=200.000000,Y=-500.000000,Z=400.000000),GravityOverride=(X=0.000000,Y=-300.000000,Z=1500.000000),JointName="root")),Skydive_FloatingDown=(GravityOverride=(WindFrequency=(X=1.500000,Y=7.000000,Z=8.000000),WindAmplitude=(X=200.000000,Y=-500.000000,Z=400.000000),WindOffset=(X=0.000000,Y=-600.000000,Z=400.000000),GravityOverride=(X=0.000000,Y=-700.000000,Z=600.000000),JointName="root")),Skydive_DivingDown=(GravityOverride=(WindFrequency=(X=1.500000,Y=12.000000,Z=10.000000),WindAmplitude=(X=50.000000,Y=300.000000,Z=100.000000),GravityOverride=(X=-1500.000000,Y=0.000000,Z=0.000000),JointName="pelvis")),Skydive_Parachuting=(GravityOverride=(WindFrequency=(X=1.500000,Y=7.000000,Z=8.000000),WindAmplitude=(X=200.000000,Y=-500.000000,Z=400.000000),GravityOverride=(X=0.000000,Y=-800.000000,Z=-980.000000),JointName="root")),OnGround_Standing=(GravityOverride=(GravityOverride=(X=0.000000,Y=0.000000,Z=-980.000000),JointName="root")),OnGround_Moving=(GravityOverride=(GravityOverride=(X=0.000000,Y=0.000000,Z=-980.000000),JointName="root")),OnGround_Moving_HighRotationRate=(GravityOverride=(GravityOverride=(X=0.000000,Y=0.000000,Z=-980.000000),JointName="root")),OnGround_Crouch_Moving=(GravityOverride=(GravityOverride=(X=0.000000,Y=0.000000,Z=-1200.000000),JointName="root")),Emote_or_Melee=(GravityOverride=(GravityOverride=(X=0.000000,Y=0.000000,Z=-980.000000),JointName="root")),Swimming=(GravityOverride=(GravityOverride=(X=-800.000000,Y=0.000000,Z=0.000000),JointName="pelvis")),Front_End=(GravityOverride=(GravityOverride=(X=0.000000,Y=0.000000,Z=-980.000000),JointName="root")),DBNO=(GravityOverride=(GravityOverride=(X=-980.000000,Y=400.000000,Z=0.000000),JointName="pelvis")),DBNO_Carried=(GravityOverride=(GravityOverride=(X=-980.000000,Y=500.000000,Z=-500.000000),JointName="pelvis")),OnGround_Falling=(GravityOverride=(GravityOverride=(X=0.000000,Y=0.000000,Z=-980.000000),JointName="root")),Ziplining=(GravityOverride=(GravityOverride=(X=0.000000,Y=0.000000,Z=-980.000000),JointName="root")),Skydive_Surfing=(GravityOverride=(WindFrequency=(X=1.500000,Y=7.000000,Z=8.000000),WindAmplitude=(X=200.000000,Y=-500.000000,Z=400.000000),GravityOverride=(X=500.000000,Y=-300.000000,Z=-980.000000),JointName="root")),Riding_Standing=(GravityOverride=(GravityOverride=(X=0.000000,Y=-350.000000,Z=-980.000000),JointName="root")),Riding_Moving=(GravityOverride=(GravityOverride=(X=0.000000,Y=-450.000000,Z=-450.000000),JointName="root")),Riding_Moving_Targeting=(GravityOverride=(GravityOverride=(X=0.000000,Y=-200.000000,Z=-980.000000),JointName="root")),Tactical_Sprint=(GravityOverride=(GravityOverride=(X=0.000000,Y=-450.000000,Z=-500.000000),JointName="root")),Rail_Grinding=(GravityOverride=(GravityOverride=(X=0.000000,Y=0.000000,Z=-980.000000),JointName="root")),)
 }
 IsEasternBlocPlaying {type=boolean}
 Local Noise Offset {type=Vector}
 Local Noise Offset_0 {type=Vector}
 "Local Noise Transform  Location" {type=Vector}
 "Mat Instance Dynamic " {type=MaterialInstanceDynamic}
 NotCrouching {type=boolean}
 "Overlya Mat Instance Dynamic " {type=MaterialInstanceDynamic}
 PlayerPawn {type=FortPlayerPawnAthena}
 R1 Velocity Transform Location {type=Vector}
 R1 Velocity Transform Velocity {type=Vector}
 R2 Velocity Transform Location {type=Vector}
 R2 Velocity Transform Velocity {type=Vector}
 RBAN Params {
  type=FortRigidBodyAnimNodeParameterSet
  default=(Skydive_GoingUp=(RigidBodyAnimNode=(SimSpaceSettings=(WorldAlpha=1.000000,MaxLinearVelocity=500.000000,MaxAngularVelocity=3.000000,MaxLinearAcceleration=800.000000,MaxAngularAcceleration=1.000000,ExternalLinearDragV=(X=1.000000,Y=1.000000,Z=1.000000)))),Skydive_FloatingDown=(RigidBodyAnimNode=(SimSpaceSettings=(WorldAlpha=1.000000,MaxLinearVelocity=500.000000,MaxAngularVelocity=3.000000,MaxLinearAcceleration=800.000000,MaxAngularAcceleration=1.000000,ExternalLinearDragV=(X=1.000000,Y=1.000000,Z=1.000000)))),Skydive_DivingDown=(RigidBodyAnimNode=(SimSpaceSettings=(WorldAlpha=1.000000,MaxLinearVelocity=500.000000,MaxAngularVelocity=3.000000,MaxLinearAcceleration=800.000000,MaxAngularAcceleration=1.000000,ExternalLinearDragV=(X=1.000000,Y=1.000000,Z=1.000000)))),Skydive_Parachuting=(RigidBodyAnimNode=(SimSpaceSettings=(WorldAlpha=1.000000,MaxLinearVelocity=500.000000,MaxAngularVelocity=3.000000,MaxLinearAcceleration=800.000000,MaxAngularAcceleration=1.000000,ExternalLinearDragV=(X=1.000000,Y=1.000000,Z=1.000000)))),OnGround_Standing=(RigidBodyAnimNode=(SimSpaceSettings=(WorldAlpha=1.000000,MaxLinearVelocity=500.000000,MaxAngularVelocity=3.000000,MaxLinearAcceleration=800.000000,MaxAngularAcceleration=1.000000,ExternalLinearDragV=(X=1.000000,Y=1.000000,Z=1.000000)))),OnGround_Moving=(RigidBodyAnimNode=(SimSpaceSettings=(WorldAlpha=1.000000,MaxLinearVelocity=500.000000,MaxAngularVelocity=3.000000,MaxLinearAcceleration=800.000000,MaxAngularAcceleration=1.000000,ExternalLinearDragV=(X=1.000000,Y=1.000000,Z=1.000000)))),OnGround_Moving_HighRotationRate=(RigidBodyAnimNode=(SimSpaceSettings=(WorldAlpha=1.000000,MaxLinearVelocity=500.000000,MaxAngularVelocity=3.000000,MaxLinearAcceleration=800.000000,MaxAngularAcceleration=1.000000,ExternalLinearDragV=(X=1.000000,Y=1.000000,Z=1.000000)))),OnGround_Crouch_Moving=(RigidBodyAnimNode=(SimSpaceSettings=(WorldAlpha=1.000000,MaxLinearVelocity=500.000000,MaxAngularVelocity=3.000000,MaxLinearAcceleration=800.000000,MaxAngularAcceleration=1.000000,ExternalLinearDragV=(X=1.000000,Y=1.000000,Z=1.000000)))),Emote_or_Melee=(RigidBodyAnimNode=(SimSpaceSettings=(WorldAlpha=1.000000,MaxLinearVelocity=500.000000,MaxAngularVelocity=3.000000,MaxLinearAcceleration=1000.000000,MaxAngularAcceleration=1.000000,ExternalLinearDragV=(X=1.000000,Y=1.000000,Z=1.000000)))),Swimming=(RigidBodyAnimNode=(SimSpaceSettings=(WorldAlpha=1.000000,MaxLinearVelocity=500.000000,MaxAngularVelocity=3.000000,MaxLinearAcceleration=800.000000,MaxAngularAcceleration=1.000000,ExternalLinearDragV=(X=1.000000,Y=1.000000,Z=1.000000)))),Front_End=(RigidBodyAnimNode=(SimSpaceSettings=(WorldAlpha=1.000000,MaxLinearVelocity=500.000000,MaxAngularVelocity=3.000000,MaxLinearAcceleration=800.000000,MaxAngularAcceleration=1.000000,ExternalLinearDragV=(X=1.000000,Y=1.000000,Z=1.000000)))),DBNO=(RigidBodyAnimNode=(SimSpaceSettings=(WorldAlpha=1.000000,MaxLinearVelocity=500.000000,MaxAngularVelocity=3.000000,MaxLinearAcceleration=800.000000,MaxAngularAcceleration=1.000000,ExternalLinearDragV=(X=1.000000,Y=1.000000,Z=1.000000)))),DBNO_Carried=(RigidBodyAnimNode=(SimSpaceSettings=(WorldAlpha=1.000000,MaxLinearVelocity=500.000000,MaxAngularVelocity=3.000000,MaxLinearAcceleration=800.000000,MaxAngularAcceleration=1.000000,ExternalLinearDragV=(X=1.000000,Y=1.000000,Z=1.000000)))),OnGround_Falling=(RigidBodyAnimNode=(SimSpaceSettings=(WorldAlpha=1.000000,MaxLinearVelocity=500.000000,MaxAngularVelocity=3.000000,MaxLinearAcceleration=800.000000,MaxAngularAcceleration=1.000000,ExternalLinearDragV=(X=1.000000,Y=1.000000,Z=1.000000)))),Ziplining=(RigidBodyAnimNode=(SimSpaceSettings=(WorldAlpha=1.000000,MaxLinearVelocity=500.000000,MaxAngularVelocity=3.000000,MaxLinearAcceleration=800.000000,MaxAngularAcceleration=1.000000,ExternalLinearDragV=(X=1.000000,Y=1.000000,Z=1.000000)))),Skydive_Surfing=(RigidBodyAnimNode=(SimSpaceSettings=(WorldAlpha=1.000000,MaxLinearVelocity=500.000000,MaxAngularVelocity=3.000000,MaxLinearAcceleration=800.000000,MaxAngularAcceleration=1.000000,ExternalLinearDragV=(X=1.000000,Y=1.000000,Z=1.000000)))),Riding_Standing=(RigidBodyAnimNode=(SimSpaceSettings=(WorldAlpha=1.000000,MaxLinearVelocity=500.000000,MaxAngularVelocity=3.000000,MaxLinearAcceleration=800.000000,MaxAngularAcceleration=1.000000,ExternalLinearDragV=(X=1.000000,Y=1.000000,Z=1.000000)))),Riding_Moving=(RigidBodyAnimNode=(SimSpaceSettings=(WorldAlpha=1.000000,MaxLinearVelocity=500.000000,MaxAngularVelocity=3.000000,MaxLinearAcceleration=800.000000,MaxAngularAcceleration=1.000000,ExternalLinearDragV=(X=1.000000,Y=1.000000,Z=1.000000)))),Riding_Moving_Targeting=(RigidBodyAnimNode=(SimSpaceSettings=(WorldAlpha=1.000000,MaxLinearVelocity=500.000000,MaxAngularVelocity=3.000000,MaxLinearAcceleration=1000.000000,MaxAngularAcceleration=1.000000,ExternalLinearDragV=(X=1.000000,Y=1.000000,Z=1.000000)))),Tactical_Sprint=(RigidBodyAnimNode=(SimSpaceSettings=(WorldAlpha=1.000000,MaxLinearVelocity=500.000000,MaxAngularVelocity=3.000000,MaxLinearAcceleration=800.000000,MaxAngularAcceleration=1.000000,ExternalLinearDragV=(X=1.000000,Y=1.000000,Z=1.000000)))),Rail_Grinding=(RigidBodyAnimNode=(SimSpaceSettings=(WorldAlpha=1.000000,MaxLinearVelocity=500.000000,MaxAngularVelocity=3.000000,MaxLinearAcceleration=800.000000,MaxAngularAcceleration=1.000000,ExternalLinearDragV=(X=1.000000,Y=1.000000,Z=1.000000)))),)
 }
 RBAN Settings Out {type=FortRigidBodyAnimNodeParameters}
 SkelMesh {type=SkeletalMeshComponent}
 Velocity Interp Speed {type=number, default=1}
 WPO Pho Drag {type=number}
 WPO Pho Turbulence {type=number}
 WPO Pho Wind Intensity {type=number}
]
