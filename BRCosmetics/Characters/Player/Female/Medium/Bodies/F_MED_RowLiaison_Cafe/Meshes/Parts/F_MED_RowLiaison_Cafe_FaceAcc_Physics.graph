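# Physics asset for the female medium RowLiaison Cafe face accessory mesh, setting up secondary motion for hair and glasses.
# Capsule bodies cover the upper spine, neck, head, clavicles and upper arms, linked by free angular constraints.
# Six three-segment hair chains and one glasses segment hang from base bodies freely constrained to the head.
# Dynamic segments have tight swing and twist limits, generally loosening toward hair tips.
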PhysicsAsset /BRCosmetics/Characters/Player/Female/Medium/Bodies/F_MED_RowLiaison_Cafe/Meshes/Parts/F_MED_RowLiaison_Cafe_FaceAcc_Physics
bodies {
 base_front_r [{shapeType=Capsule, length=0.505, radius=0.505, shapeName=base_front_r_capsule, rotation=<rot 0 -90 0>}]
 base_glasses [{shapeType=Capsule, length=3.0399, radius=5.6426, shapeName=base_glasses_capsule, center=<4.2447 0 0>, rotation=<rot 0 -90 0>}]
 base_hair_l [{shapeType=Capsule, length=0.505, radius=0.505, shapeName=base_hair_l_capsule, rotation=<rot 0 -90 0>}]
 base_hair_m [{shapeType=Capsule, length=0.505, radius=0.505, shapeName=base_hair_m_capsule, rotation=<rot 0 -90 0>}]
 base_hair_r [{shapeType=Capsule, length=0.505, radius=0.505, shapeName=base_hair_r_capsule, rotation=<rot 0 -90 0>}]
 base_side_l [{shapeType=Capsule, length=0.505, radius=0.505, shapeName=base_side_l_capsule, rotation=<rot 0 -90 0>}]
 base_side_r [{shapeType=Capsule, length=0.505, radius=0.505, shapeName=base_side_r_capsule, rotation=<rot 0 -90 0>}]
 clavicle_l [{
   shapeType=Capsule
   length=8.0319
   radius=4
   shapeName=clavicle_l_capsule
   center=<5.024 -0.5885 0.0917>
   rotation=<rot 81.839 -0.0005 -0.0005>
  }]
 clavicle_r [{
   shapeType=Capsule
   length=8.3741
   radius=4.8343
   shapeName=clavicle_r_capsule
   center=<-4.5939 1.0938 1.0166>
   rotation=<rot 83.3961 -0.001 179.999>
  }]
 dyn_front_r_01 [{shapeType=Capsule, length=3, radius=2, shapeName=dyn_front_r_01_capsule, center=<2 0 0>, rotation=<rot 90 0 0>}]
 dyn_front_r_02 [{shapeType=Capsule, length=3, radius=2, shapeName=dyn_front_r_02_capsule, center=<2 0 0>, rotation=<rot 90 0 0>}]
 dyn_front_r_03 [{shapeType=Capsule, length=2.7858, radius=2.2142, shapeName=dyn_front_r_03_capsule, center=<2 0 0>, rotation=<rot 90 0 0>}]
 dyn_glasses [{shapeType=Capsule, length=3, radius=2, shapeName=dyn_glasses_capsule, center=<3 0 0>, rotation=<rot 90 0 0>}]
 dyn_hair_l_01 [{shapeType=Capsule, length=3, radius=2, shapeName=dyn_hair_l_01_capsule, center=<2 0 0>, rotation=<rot 90 0 0>}]
 dyn_hair_l_02 [{shapeType=Capsule, length=3, radius=2, shapeName=dyn_hair_l_02_capsule, center=<2 0 0>, rotation=<rot 90 0 0>}]
 dyn_hair_l_03 [{shapeType=Capsule, length=2.53, radius=2.47, shapeName=dyn_hair_l_03_capsule, center=<2 0 0>, rotation=<rot 90 0 0>}]
 dyn_hair_m_01 [{shapeType=Capsule, length=3, radius=2, shapeName=dyn_hair_m_01_capsule, center=<2 0 0>, rotation=<rot 90 0 0>}]
 dyn_hair_m_02 [{shapeType=Capsule, length=3, radius=2, shapeName=dyn_hair_m_02_capsule, center=<2 0 0>, rotation=<rot 90 0 0>}]
 dyn_hair_m_03 [{shapeType=Capsule, length=3, radius=2, shapeName=dyn_hair_m_03_capsule, center=<2 0 0>, rotation=<rot 90 0 0>}]
 dyn_hair_r_01 [{shapeType=Capsule, length=3, radius=2, shapeName=dyn_hair_r_01_capsule, center=<2 0 0>, rotation=<rot 90 0 0>}]
 dyn_hair_r_02 [{shapeType=Capsule, length=3, radius=2, shapeName=dyn_hair_r_02_capsule, center=<2 0 -0.3378>, rotation=<rot 90 0 0>}]
 dyn_hair_r_03 [{shapeType=Capsule, length=4.3938, radius=2.9451, shapeName=dyn_hair_r_03_capsule, center=<1.6 -1 -2.3>, rotation=<rot 70 -180 -180>}]
 dyn_side_l_01 [{shapeType=Capsule, length=3, radius=2, shapeName=dyn_side_l_01_capsule, center=<2 0 0>, rotation=<rot 90 0 0>}]
 dyn_side_l_02 [{shapeType=Capsule, length=3, radius=2, shapeName=dyn_side_l_02_capsule, center=<2 0 0>, rotation=<rot 90 0 0>}]
 dyn_side_l_03 [{shapeType=Capsule, length=3, radius=2, shapeName=dyn_side_l_03_capsule, center=<2 0 0>, rotation=<rot 90 0 0>}]
 dyn_side_r_01 [{shapeType=Capsule, length=3, radius=2, shapeName=dyn_side_r_01_capsule, center=<2 0 0>, rotation=<rot 90 0 0>}]
 dyn_side_r_02 [{shapeType=Capsule, length=3, radius=2, shapeName=dyn_side_r_02_capsule, center=<2 0 0>, rotation=<rot 90 0 0>}]
 dyn_side_r_03 [{shapeType=Capsule, length=2.8176, radius=2.1824, shapeName=dyn_side_r_03_capsule, center=<2.1255 0 -0.512>, rotation=<rot 90 0 0>}]
 head [{shapeType=Capsule, length=0.44, radius=5.8707, shapeName=head_capsule, center=<0.7982 0.8112 0>}]
 neck_01 [{shapeType=Capsule, length=0.5295, radius=4.6745, shapeName=neck_01_capsule, center=<0 0.4185 0>}]
 neck_02 [{shapeType=Capsule, length=0.1602, radius=3.4324, shapeName=neck_02_capsule, rotation=<rot 0 -90 0>}]
 spine_05 [
  {shapeType=Capsule, length=0, radius=9.1526, shapeName=spine_05_capsule, center=<3.455 0.469 0>, rotation=<rot 0 -90 0>}
  {shapeType=Capsule, length=7.3718, radius=9.8741, shapeName=spine_05_capsule, center=<-1.1311 1.6378 0>}
 ]
 upperarm_l [{
   shapeType=Capsule
   length=18.1201
   radius=4
   shapeName=upperarm_l_capsule
   center=<9.0792 -0.4928 -0.5366>
   rotation=<rot 89.1653 -0.554 1.0035>
  }]
 upperarm_r [{
   shapeType=Capsule
   length=21.9782
   radius=3
   shapeName=upperarm_r_capsule
   center=<-8.7703 1.2721 0.8049>
   rotation=<rot 89.8676 172.1179 -6.3248>
  }]
}
constraints [
 {
  bone1Name=head
  bone2Name=neck_02
  swing1LimitDegrees=45
  swing1Motion=Free
  swing2LimitDegrees=45
  swing2Motion=Free
  twistLimitDegrees=45
  twistMotion=Free
 }
 {
  bone1Name=neck_02
  bone2Name=neck_01
  swing1LimitDegrees=45
  swing1Motion=Free
  swing2LimitDegrees=45
  swing2Motion=Free
  twistLimitDegrees=45
  twistMotion=Free
 }
 {
  bone1Name=base_glasses
  bone2Name=head
  swing1LimitDegrees=45
  swing1Motion=Free
  swing2LimitDegrees=45
  swing2Motion=Free
  twistLimitDegrees=45
  twistMotion=Free
 }
 {bone1Name=dyn_glasses, bone2Name=base_glasses, swing1LimitDegrees=5, swing2LimitDegrees=0, twistLimitDegrees=0}
 {
  bone1Name=base_front_r
  bone2Name=head
  swing1LimitDegrees=45
  swing1Motion=Free
  swing2LimitDegrees=45
  swing2Motion=Free
  twistLimitDegrees=45
  twistMotion=Free
 }
 {bone1Name=dyn_front_r_01, bone2Name=base_front_r, swing1LimitDegrees=10, swing2LimitDegrees=20, twistLimitDegrees=5}
 {bone1Name=dyn_front_r_02, bone2Name=dyn_front_r_01, swing1LimitDegrees=15, swing2LimitDegrees=25, twistLimitDegrees=5}
 {bone1Name=dyn_front_r_03, bone2Name=dyn_front_r_02, swing1LimitDegrees=25, swing2LimitDegrees=30, twistLimitDegrees=5}
 {
  bone1Name=base_hair_l
  bone2Name=head
  swing1LimitDegrees=45
  swing1Motion=Free
  swing2LimitDegrees=45
  swing2Motion=Free
  twistLimitDegrees=45
  twistMotion=Free
 }
 {bone1Name=dyn_hair_l_01, bone2Name=base_hair_l, swing1LimitDegrees=10, swing2LimitDegrees=20, twistLimitDegrees=5}
 {bone1Name=dyn_hair_l_02, bone2Name=dyn_hair_l_01, swing1LimitDegrees=15, swing2LimitDegrees=15, twistLimitDegrees=5}
 {bone1Name=dyn_hair_l_03, bone2Name=dyn_hair_l_02, swing1LimitDegrees=25, swing2LimitDegrees=30, twistLimitDegrees=5}
 {
  bone1Name=base_hair_m
  bone2Name=head
  swing1LimitDegrees=45
  swing1Motion=Free
  swing2LimitDegrees=45
  swing2Motion=Free
  twistLimitDegrees=45
  twistMotion=Free
 }
 {bone1Name=dyn_hair_m_01, bone2Name=base_hair_m, swing1LimitDegrees=5, swing2LimitDegrees=20, twistLimitDegrees=5}
 {bone1Name=dyn_hair_m_02, bone2Name=dyn_hair_m_01, swing1LimitDegrees=5, swing2LimitDegrees=15, twistLimitDegrees=5}
 {bone1Name=dyn_hair_m_03, bone2Name=dyn_hair_m_02, swing1LimitDegrees=10, swing2LimitDegrees=30, twistLimitDegrees=5}
 {
  bone1Name=base_hair_r
  bone2Name=head
  swing1LimitDegrees=45
  swing1Motion=Free
  swing2LimitDegrees=45
  swing2Motion=Free
  twistLimitDegrees=45
  twistMotion=Free
 }
 {bone1Name=dyn_hair_r_01, bone2Name=base_hair_r, swing1LimitDegrees=10, swing2LimitDegrees=20, twistLimitDegrees=5}
 {bone1Name=dyn_hair_r_02, bone2Name=dyn_hair_r_01, swing1LimitDegrees=10, swing2LimitDegrees=15, twistLimitDegrees=5}
 {bone1Name=dyn_hair_r_03, bone2Name=dyn_hair_r_02, swing1LimitDegrees=10, swing2LimitDegrees=30, twistLimitDegrees=5}
 {
  bone1Name=base_side_l
  bone2Name=head
  swing1LimitDegrees=45
  swing1Motion=Free
  swing2LimitDegrees=45
  swing2Motion=Free
  twistLimitDegrees=45
  twistMotion=Free
 }
 {bone1Name=dyn_side_l_01, bone2Name=base_side_l, swing1LimitDegrees=10, swing2LimitDegrees=20, twistLimitDegrees=5}
 {bone1Name=dyn_side_l_02, bone2Name=dyn_side_l_01, swing1LimitDegrees=15, swing2LimitDegrees=35, twistLimitDegrees=5}
 {bone1Name=dyn_side_l_03, bone2Name=dyn_side_l_02, swing1LimitDegrees=25, swing2LimitDegrees=30, twistLimitDegrees=5}
 {
  bone1Name=base_side_r
  bone2Name=head
  swing1LimitDegrees=45
  swing1Motion=Free
  swing2LimitDegrees=45
  swing2Motion=Free
  twistLimitDegrees=45
  twistMotion=Free
 }
 {bone1Name=dyn_side_r_01, bone2Name=base_side_r, swing1LimitDegrees=10, swing2LimitDegrees=20, twistLimitDegrees=5}
 {bone1Name=dyn_side_r_02, bone2Name=dyn_side_r_01, swing1LimitDegrees=15, swing2LimitDegrees=25, twistLimitDegrees=5}
 {bone1Name=dyn_side_r_03, bone2Name=dyn_side_r_02, swing1LimitDegrees=25, swing2LimitDegrees=30, twistLimitDegrees=5}
 {
  bone1Name=neck_01
  bone2Name=spine_05
  swing1LimitDegrees=45
  swing1Motion=Free
  swing2LimitDegrees=45
  swing2Motion=Free
  twistLimitDegrees=45
  twistMotion=Free
 }
 {
  bone1Name=clavicle_l
  bone2Name=spine_05
  swing1LimitDegrees=45
  swing1Motion=Free
  swing2LimitDegrees=45
  swing2Motion=Free
  twistLimitDegrees=45
  twistMotion=Free
 }
 {
  bone1Name=clavicle_r
  bone2Name=spine_05
  swing1LimitDegrees=45
  swing1Motion=Free
  swing2LimitDegrees=45
  swing2Motion=Free
  twistLimitDegrees=45
  twistMotion=Free
 }
 {
  bone1Name=upperarm_l
  bone2Name=clavicle_l
  swing1LimitDegrees=45
  swing1Motion=Free
  swing2LimitDegrees=45
  swing2Motion=Free
  twistLimitDegrees=45
  twistMotion=Free
 }
 {
  bone1Name=upperarm_r
  bone2Name=clavicle_r
  swing1LimitDegrees=45
  swing1Motion=Free
  swing2LimitDegrees=45
  swing2Motion=Free
  twistLimitDegrees=45
  twistMotion=Free
 }
]
